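# Sound cue for collecting a Creative power-up toy, layering two reused ingredient-collect foley waves.
# A Mixer blends a powder-collect wave, shaped by an Enveloper with slight pitch/volume variation and a fade-out, with a quieter surreal plant-collect wave after a short fixed delay.
# It plays at reduced volume on the non-critical in-game UI sound class, overriding attenuation for short-range falloff and concurrency to cap simultaneous instances.
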SoundCue /Game/Sounds/Creative/Toys/PowerUp/Creative_PowerUp_Collect_Cue
root=Mixer_0
nodes [
 {id=Delay_1, children [WavePlayer_3], properties {delayMax=0.13, delayMin=0.13}}
 {
  id=Enveloper_0
  children [WavePlayer_6]
  properties {
   pitchMin=0.95
   volumeMin=0.95
   pitchCurve {editorCurveData {keys [{time=0}]}}
   volumeCurve {editorCurveData {keys [
      {arriveTangent=0.2393, time=0, value=1}
      {arriveTangent=-0.8, interpMode=RCIM_Cubic, leaveTangent=-0.8, tangentMode=RCTM_User, time=0.6744, value=1}
      {interpMode=RCIM_Cubic, tangentMode=RCTM_User, time=1.2, value=0}
     ]}}
  }
 }
 {id=Mixer_0, children [Enveloper_0,Delay_1], properties {inputVolume [1.2,0.9]}}
 {id=WavePlayer_3, properties {soundWaveAssetPtr=/Game/Sounds/Foley_Loot/Ingredients/Health/Plant_Collect_Surreal02}}
 {id=WavePlayer_6, properties {soundWaveAssetPtr=/Game/Sounds/Foley_Loot/Ingredients/Munitions/Powder_Collect_Real}}
]
properties {
 bOverrideAttenuation=true
 bOverrideConcurrency=true
 maxDistance=727
 soundClassObject=&Fort_UI_InGame_NonCritical
 volumeMultiplier=0.5
 attenuationOverrides {falloffDistance=512, attenuationShapeExtents {x=215}}
 concurrencyOverrides {maxCount=12, volumeScale=0.7}
}
