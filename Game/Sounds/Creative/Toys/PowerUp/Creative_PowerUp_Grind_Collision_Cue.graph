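# Sound cue for the collision sound of a grind power-up among the Creative toys.
# A random node picks one of two grind collision waves, and a mixer layers it with a quieter electric spark zap borrowed from the trap sounds.
# A modulator at the root varies the mix, which plays at low volume with small melee-weapon attenuation, a melee weapon sound class and owner-limited concurrency.
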SoundCue /Game/Sounds/Creative/Toys/PowerUp/Creative_PowerUp_Grind_Collision_Cue
root=Modulator_0
nodes [
 {id=Mixer_0, children [Random_0,WavePlayer_5], properties {inputVolume [1,0.25]}}
 {id=Modulator_0, children [Mixer_0]}
 {id=Random_0, children [WavePlayer_4,WavePlayer_2], properties {weights [1,1]}}
 {id=WavePlayer_2, properties {soundWaveAssetPtr=&/PowerUp_Grind_Collision_01}}
 {id=WavePlayer_4, properties {soundWaveAssetPtr=&/PowerUp_Grind_Collision_03}}
 {id=WavePlayer_5, properties {soundWaveAssetPtr=/Game/Sounds/Fort_Traps/Transformer/ElectricitySpark_Zap_2}}
]
properties {
 attenuationSettings=&Fort_Weapon_Melee_Sml_Att
 maxDistance=2800
 soundClassObject=&Fort_WeaponMelee_PSM
 volumeMultiplier=0.2
 concurrencyOverrides {bLimitToOwner=true}
}
