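# Sound cue for the grind loop of a Creative toy power-up.
# It plays one of two looping grind waves at random, with pitch and volume continuously driven by a Speed parameter.
# A Mixer layers this over a much quieter, pitched-up, non-looping electric impact.
# It uses a footstep attenuation asset and an enemy foley sound class, plus occlusion and limit-to-owner concurrency settings.
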SoundCue /Game/Sounds/Creative/Toys/PowerUp/Creative_PowerUp_Grind_Loop_Cue
root=Mixer_1
nodes [
 {id=Mixer_1, children [ModulatorContinuous_1, Modulator_0], properties {inputVolume [1,0.1]}}
 {
  id=ModulatorContinuous_1
  children [Random_0]
  properties {
   pitchModulationParams {minInput=0.97, minOutput=0.97, parameterName=Speed}
   volumeModulationParams {default=0.6, maxOutput=1.1, minInput=0.3, minOutput=0.01, parameterName=Speed}
  }
 }
 {id=Modulator_0, children [WavePlayer_4], properties {pitchMax=1.25, pitchMin=1.15}}
 {id=Random_0, children [WavePlayer_1,WavePlayer_0], properties {weights [1,1]}}
 {id=WavePlayer_0, properties {bLooping=true, soundWaveAssetPtr=&/PowerUp_Grind_Loop_01}}
 {id=WavePlayer_1, properties {bLooping=true, soundWaveAssetPtr=&/PowerUp_Grind_Loop_02}}
 {id=WavePlayer_4, properties {soundWaveAssetPtr=/Game/Sounds/Fort_GamePlay_Sounds/Elemental_Layers/enemy_electric_small_impact_06}}
]
properties {
 attenuationSettings=/Game/Sounds/Attenuation/Fort_Player_Footsteps_1P_Walk_Athena_Att
 maxDistance=2097152
 soundClassObject=&Fort_Player_Foley_3P_Enemy_Athena
 volumeMultiplier=0.4762
 attenuationOverrides {
  bEnableOcclusion=true
  lPFFrequencyAtMin=10000
  lPFRadiusMin=2000
  nonFocusVolumeAttenuation=0.7
  occlusionLowPassFilterFrequency=4000
  occlusionTraceChannel=ECC_GameTraceChannel11
  occlusionVolumeAttenuation=4
 }
 concurrencyOverrides {bLimitToOwner=true}
}
