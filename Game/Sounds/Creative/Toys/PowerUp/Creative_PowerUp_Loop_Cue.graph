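# Looping sound cue for a Creative mode power-up toy, played as a short-range ambient loot loop.
# A mixer blends three looping layers: a store power-up loop, a quieter pitched-up sword energy loop and a dominant pitched-up low-gravity stone loop.
# A root modulator varies the blend, and the cue raises overall pitch.
# Overridden logarithmic attenuation adds listener focus, distance low-pass and occlusion settings, with a shared loot ambient concurrency set.
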SoundCue /Game/Sounds/Creative/Toys/PowerUp/Creative_PowerUp_Loop_Cue
root=Modulator_2
nodes [
 {id=Mixer_1, children [WavePlayer_5,Modulator_0,Modulator_1], properties {inputVolume [0.35,0.2,1]}}
 {id=Modulator_0, children [WavePlayer_6], properties {pitchMax=1.3, pitchMin=1.2}}
 {id=Modulator_1, children [WavePlayer_4], properties {pitchMax=1.2, pitchMin=1.1}}
 {id=Modulator_2, children [Mixer_1]}
 {
  id=WavePlayer_4
  properties {bLooping=true, soundWaveAssetPtr=/Game/Sounds/Fort_GamePlay_Sounds/Player/LowGravityStones/LowGravityStone_InWorld_Loop}
 }
 {id=WavePlayer_5, properties {bLooping=true, soundWaveAssetPtr=/Game/Sounds/Fort_UI_Sounds/FrontEnd/Store/wav/powerup_loop}}
 {id=WavePlayer_6, properties {bLooping=true, soundWaveAssetPtr=/Game/Sounds/Fort_GamePlay_Sounds/BRSword/AB_Pull_Energy_Loop_01}}
]
properties {
 bOverrideAttenuation=true
 maxDistance=600
 pitchMultiplier=1.2
 soundClassObject=/Game/Sounds/FortSoundClassesAndModes/Fort_Loot_Ambient_Loop
 volumeMultiplier=0.45
 attenuationOverrides {
  bEnableFocusInterpolation=true
  bEnableListenerFocus=true
  distanceAlgorithm=Logarithmic
  falloffDistance=500
  focusAttackInterpSpeed=20
  focusAzimuth=20
  focusReleaseInterpSpeed=10
  lPFFrequencyAtMax=10000
  lPFRadiusMax=1000
  lPFRadiusMin=500
  nonFocusAzimuth=40
  nonFocusVolumeAttenuation=0.6
  occlusionInterpolationTime=0.3
  occlusionLowPassFilterFrequency=3000
  occlusionVolumeAttenuation=0.75
  attenuationShapeExtents {x=100}
  customAttenuationCurve {editorCurveData {keys [<at 0 1 RCIM_Cubic>, {arriveTangent=-1, interpMode=RCIM_Cubic, leaveTangent=-1, time=0.5, value=0.1}, <at 1 0 RCIM_Cubic>]}}
 }
 concurrencyOverrides {bVolumeScaleCanRelease=true, maxCount=3, volumeScaleAttackTime=0.1, volumeScaleReleaseTime=0.15}
 concurrencySet [/Game/Sounds/Concurrency_Rules/Loot_Ambient_Conc]
}
